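# Material function for a loot-container beam FX sprite that turns quads toward the camera along a direction vector (local Y+ by default) via World Position Offset.
# Cross products build the sprite axes, with axis lock on by default, and pivot-adjusted UVs times scale rebuild vertices around the center.
# It chains additional WPO and pushes vertices along their camera rays to keep screen size stable.
# A second output re-bases the input normal into tangent space.
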MaterialFunction /Train/LootContainers/Meshes/FX/MF_Hermes_Beam_Sprite_Directional
graph {nodes [
  n1 = FunctionInput(Preview=Transform(None=Constant3Vector() {constant=<rgba 0 1 0 0>}) {transformSourceType=TRANSFORMSOURCE_Local}) {
   bUsePreviewValueAsDefault=true
   description=Vector to point the sprite towards. Defaults to local Y+
   inputName=Direction Vector
   sortPriority=4
   previewValue=<0 0 0 0>
  }
  n2 = Normalize(VectorInput=@n1)
  n3 = FunctionInput(Preview=MaterialFunctionCall() {materialFunction=&ComponentPivotLocation}.Result) {
   bUsePreviewValueAsDefault=true
   description=The world position of the center of the sprite
   inputName=Center Location
   sortPriority=5
   previewValue=<0 0 0 0>
  }
  n4 = Normalize(VectorInput=Subtract(A=CameraPositionWS(), B=@n3))
  n5 = Normalize(VectorInput=CrossProduct(A=@n2, B=@n4))
  n6 = FunctionInput() {
   bUsePreviewValueAsDefault=true
   description="Tangent Space, optional, transforms a Normal Map so it works correctly."
   inputName=Normal
   sortPriority=7
   previewValue=<0 0 1 1>
  }
  n7 = MaterialFunctionCall(Float3=@n6) {materialFunction=&BreakOutFloat3Components}
  n8 = FunctionInput(Preview=StaticBool() {value=true}) {
   bUsePreviewValueAsDefault=true
   description=Will constrain sprite's rotation around the length axis. Defaults to True
   inputName=Axis Locked
   inputType=FunctionInput_StaticBool
   sortPriority=3
  }
  n9 = StaticSwitch(True=@n2, False=CrossProduct(A=@n4, B=@n5), Value=@n8) {defaultValue=true}
  n10 = Add(A=Multiply(A=@n5, B=@n7.R), B=Multiply(A=@n9, B=@n7.G))
  n11 = Multiply(A=CrossProduct(A=@n5, B=@n9), B=@n7.B)
  FunctionOutput(None=Transform(None=Add(A=@n10, B=@n11)) {transformSourceType=TRANSFORMSOURCE_World, transformType=TRANSFORM_Tangent}) {description="Transformed Normal Input, in Tangent Space", outputName=Normal, sortPriority=1}
  n12 = TextureCoordinate()
  n13 = FunctionInput() {bUsePreviewValueAsDefault=true, description=In UV Space, inputName=Pivot Offset, inputType=FunctionInput_Vector2, sortPriority=2}
  n14 = MaterialFunctionCall() {materialFunction=&ObjectLocalBounds}
  n15 = MaterialFunctionCall() {materialFunction=/Engine/Functions/Engine_MaterialFunctions02/WorldPositionOffset/ObjectScale}
  n16 = FunctionInput(Preview=Multiply(A=ComponentMask(None=@n14.Local Bounds Size) {g=true, r=true}, B=ComponentMask(None=@n15.Scale XYZ) {g=true, r=true})) {
   bUsePreviewValueAsDefault=true
   description=Defaults to object's local bounds * scale
   inputName=XY Scale
   inputType=FunctionInput_Vector2
   sortPriority=1
   previewValue=<111 111 0 1>
  }
  n17 = Multiply(
   A=Subtract(
    A=FunctionInput(Preview=@n12) {bUsePreviewValueAsDefault=true, description=Clamped 0-1, inputName=UV, inputType=FunctionInput_Vector2}
    B=Add(A=@n13) {constB=0.5}
   )
   B=@n16
  )
  n18 = Multiply(A=ComponentMask(None=@n17) {r=true}, B=@n5)
  n19 = Multiply(A=ComponentMask(None=@n17) {g=true}, B=@n9)
  n20 = WorldPosition()
  n21 = FunctionInput() {
   bUsePreviewValueAsDefault=true
   description="Additional WPO, optional, allows you to chain other WPO operations"
   inputName=Additional World Position Offset
   sortPriority=6
  }
  n22 = Add(A=@n20.XYZ, B=@n21)
  n23 = Add(A=Add(A=Add(A=@n18, B=@n19), B=Subtract(A=@n3, B=@n22)), B=Multiply(A=@n21) {constB=2}) {desc=To World Position Offset}
  n24 = Normalize(VectorInput=Subtract(A=CameraPositionWS(), B=Add(A=@n23, B=@n20.XYZ)))
  n25 = FunctionInput() {
   bUsePreviewValueAsDefault=true
   description="Pushes the sprite towards (+) or away (-) from camera, while keeping the screen size stable"
   inputName=Camera Offset
   inputType=FunctionInput_Scalar
   sortPriority=8
  }
  FunctionOutput(None=Add(A=@n23, B=Multiply(A=@n24, B=@n25))) {description=World Position Offset, outputName=World Position Offset}
 ]}
