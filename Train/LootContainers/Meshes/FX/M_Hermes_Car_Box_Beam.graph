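# Additive unlit light-beam material for the Hermes car box in the train loot container FX, driven by CapturePointState custom primitive data.
# A directional sprite function builds an axis-locked beam with view-dependent width, scaled to zero at state 3.
# Panning noise, edge and length falloffs, an orientation-based fresnel, depth fade and distance-cull fade form the opacity.
# The floored state switches emissive color from idle to activated.
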Material /Train/LootContainers/Meshes/FX/M_Hermes_Car_Box_Beam
graph {
 properties {bAutomaticallySetUsageInEditor=false, blendMode=BLEND_Additive, shadingModel=MSM_Unlit, usedShadingModels=Unlit}
 nodes [
  n1 = Transform(None=Constant3Vector() {constant=<rgba 1 0 0 0>})
  n2 = LocalPosition()
  n3 = Constant3Vector() {constant=<rgba 0 0 1 0>}
  n4 = Subtract(A=CameraPositionWS(), B=TransformPosition(None=Multiply(A=@n2.XYZ, B=@n3)) {transformType=TRANSFORMPOSSOURCE_World})
  n5 = ComponentMask(None=@n4) {g=true, r=true}
  n6 = Abs(None=DotProduct(A=ComponentMask(None=@n1) {g=true, r=true}, B=Normalize(VectorInput=@n5)))
  n7 = LinearInterpolate(
   A=ScalarParameter() {defaultValue=44, parameterName=Width Side}
   B=ScalarParameter() {defaultValue=99, parameterName=Width Front}
   Alpha=SquareRoot(None=OneMinus(None=@n6))
  )
  n8 = ScalarParameter() {defaultValue=700, parameterName=Height}
  n9 = ScalarParameter() {bUseCustomPrimitiveData=true, parameterName=CapturePointState}
  n10 = Constant() {r=1}
  n11 = Constant()
  MaterialFunctionCall!(
   UV
   XY Scale=Multiply(A=AppendVector(A=@n7, B=@n8), B=If(A=@n9, "A > B"=@n10, "A == B"=@n11, "A < B"=@n10) {constB=3})
   Pivot Offset=Constant2Vector() {g=0.5}
   Axis Locked
   Direction Vector=Constant3Vector() {constant=<rgba 0 0 -1 0>}
   Center Location
   Additional World Position Offset
   Normal
   Camera Offset=ScalarParameter() {defaultValue=11, parameterName=Camera Offset}
  ) {materialFunction=&/MF_Hermes_Beam_Sprite_Directional}
  MaterialFunctionCall!(World Position, DistanceCullFadeAnimationOffset, Use actor center instead of pivot point) {materialFunction=/Game/Packages/Fortress_SharedMaterials/Base_Material_Functions/LODEffect}
  n12 = AppendVector(A=ComponentMask(None=TextureCoordinate()) {r=true}, B=Constant())
  n13 = ScalarParameter() {defaultValue=0.3, parameterName=Noise UV Scale}
  n14 = ScalarParameter() {defaultValue=-0.03, parameterName=Noise Panner Speed U}
  n15 = ScalarParameter() {defaultValue=0.16, parameterName=Noise Panner Speed V}
  n16 = TextureSample(UVs=Panner(Coordinate=Multiply(A=@n12, B=@n13), Speed=AppendVector(A=@n14, B=@n15))) {samplerType=SAMPLERTYPE_LinearGrayscale, texture=/Game/Effects/Fort_Effects/Emotes/UltraLight/Texture/T_DirectionalNoise_Greyscale}
  n17 = OneMinus(None=Multiply(A=OneMinus(None=@n16.R), B=ScalarParameter() {defaultValue=0.5, parameterName=Noise Intensity}))
  n18 = TextureCoordinate()
  n19 = ConstantBiasScale(None=ComponentMask(None=@n18) {r=true}) {bias=-0.5, scale=2}
  n20 = Saturate(None=DotProduct(A=@n19, B=@n19))
  n21 = Saturate(None=SquareRoot(None=OneMinus(None=@n20)))
  n22 = ScalarParameter() {defaultValue=4, parameterName=Edge Falloff}
  n23 = ComponentMask(None=@n18) {g=true}
  n24 = ScalarParameter() {defaultValue=2, parameterName=Length Falloff}
  n25 = ObjectOrientation()
  n26 = Normalize(VectorInput=@n4)
  n27 = OneMinus(None=Abs(None=DotProduct(A=@n25, B=@n26)))
  n28 = ScalarParameter() {defaultValue=1, parameterName=Fresnel Exponent}
  n29 = Multiply(
   A=Multiply(A=Power(Base=@n21, Exp=@n22), B=Power(Base=@n23, Exp=@n24))
   B=DepthFade(Opacity=Power(Base=@n27, Exp=@n28), FadeDistance=ScalarParameter() {defaultValue=11, parameterName=Depth Fade Distance})
  )
  NamedRerouteDeclaration(None=Multiply(A=Multiply(A=@n17, B=@n29), B=DistanceCullFade())) {name=Opacity, nodeColor=<rgba 1 0.2812 0 1>}
  n30 = VectorParameter() {parameterName=Color Idle, defaultValue=<rgba 2 1 0 0>}
  n31 = VectorParameter() {parameterName=Color Activated, defaultValue=<rgba 0 2 2 0>}
  n32 = Saturate(None=Subtract(A=Floor(None=@n9)))
  NamedRerouteDeclaration(None=MaterialFunctionCall(Emissive=LinearInterpolate(A=@n30.RGB, B=@n31.RGB, Alpha=@n32)) {materialFunction=&MF_FX_EmissiveAdjustments}.Result) {name=Emissive Color, nodeColor=<rgba 0 0.0391 1 1>}
  WorldPosition()
 ]
}
