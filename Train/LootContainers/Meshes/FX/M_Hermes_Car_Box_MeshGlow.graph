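# Additive unlit glow material for the Hermes train car loot box mesh, driven by the CapturePointState custom primitive data value.
# A power-shaped gradient along UV U blends black through two colors, which switch from idle orange to activated blue as the floored state rises.
# World position offset applies normal and camera offsets, but collapses vertices to the object pivot when the state equals 3, effectively hiding the glow.
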Material /Train/LootContainers/Meshes/FX/M_Hermes_Car_Box_MeshGlow
graph {
 properties {bAutomaticallySetUsageInEditor=false, blendMode=BLEND_Additive, shadingModel=MSM_Unlit, usedShadingModels=Unlit}
 nodes [
  CameraPositionWS()
  n1 = VertexNormalWS()
  n2 = ScalarParameter() {parameterName=Normal Offset}
  n3 = ScalarParameter() {defaultValue=0.5, parameterName=Camera Offset}
  n4 = ObjectPositionWS()
  n5 = WorldPosition()
  n6 = ScalarParameter() {bUseCustomPrimitiveData=true, parameterName=CapturePointState}
  n7 = Constant()
  n8 = If(A=@n6, "A > B"=@n7, "A == B"=Constant() {r=1}, "A < B"=@n7) {constB=3}
  LinearInterpolate(
   A=Add(
    A=Multiply(A=@n1, B=@n2)
    B=MaterialFunctionCall(Offset Amount=@n3) {materialFunction=/Engine/Functions/Engine_MaterialFunctions02/WorldPositionOffset/CameraOffset}.Camera Offset
   )
   B=Multiply(A=Subtract(A=@n4, B=@n5.XYZ), B=@n8)
   Alpha=@n8
  )
  n9 = VectorParameter() {parameterName=Color Idle B, defaultValue=<rgba 1 0.5 0 0>}
  n10 = VectorParameter() {parameterName=Color Activated B, defaultValue=<rgba 0 0.5667 1 0>}
  n11 = Saturate(None=Subtract(A=Floor(None=@n6)))
  n12 = VectorParameter() {parameterName=Color Idle A, defaultValue=<rgba 1.5 0.9 0.3 0>}
  n13 = VectorParameter() {parameterName=Color Activated A, defaultValue=<rgba 0.3 0.98 1.5 0>}
  n14 = ComponentMask(None=TextureCoordinate()) {r=true}
  n15 = ScalarParameter() {defaultValue=2, parameterName=Gradient Exponent}
  MaterialFunctionCall(Emissive=MaterialFunctionCall(
    1=Constant3Vector()
    2=LinearInterpolate(A=@n9.RGB, B=@n10.RGB, Alpha=@n11)
    3=LinearInterpolate(A=@n12.RGB, B=@n13.RGB, Alpha=@n11)
    A=Power(Base=@n14, Exp=@n15)
   ) {materialFunction=&Lerp_Multiple_Float3}.Lerp 3 Inputs) {materialFunction=&MF_FX_EmissiveAdjustments}
 ]
}
